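# Niagara system that draws ribbon swing trails for the DirectContact pickaxe cosmetic.
# TrailBlade and TrailSphere spawn anim-trail particles between separate bone or socket pairs on the system's skeletal mesh, with short lifetimes.
# Each sets its tint and opacity by time of day, then fades alpha over particle life and again by emitter age, cutting the trail off early.
# TrailBlade's inherit-movement and force-solver modules are disabled.
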NiagaraSystem /BRCosmetics/Weapons/FORT_Melee/Pickaxe_DirectContact/FX/NS_Pickaxe_DirectContact_Trails
dependencies {
 usedDataInterfaces [&NiagaraDataInterfaceColorCurve, &NiagaraDataInterfaceCurve, &NiagaraDataInterfaceSkeletalMesh]
 usedDynamicInputs []
 usedModules [
  &self:SystemSpawnScript.SystemScriptSource.SystemScriptGraph.NiagaraNodeAssignment_0.SetVariables_#1
  &SystemState
 ]
}
emitters {
 TrailBlade {
  emitterUpdateScript [
   &EmitterState()
   &FN_SpawnAnimTrail(Spawn Interval=0)
  ]
  particleSpawnScript [
   &FN_InitializeAnimTrailBoneIndecies(First Bone or Socket Index=1, Second Bone or Socket Index=0)
   &FN_InitializeAnimTrail(AnimTrailWidth=1.1, Lifetime=0.12, SkeletalMesh=$System.Skeletal Mesh)
   /SetVariables(Particles.Color=&FNTimeOfDayColor(Night Color=<rgba 0.7 0.75 1 0.3>, Morning Color=<rgba 0.7 0.75 1 0.3>, Day Color=<rgba 0.7 0.75 1 0.4>, Evening Color=<rgba 0.7 0.75 1 0.3>)) {isSetParameters=true}
  ]
  particleUpdateScript [
   &FN_UpdateAnimTrail(Locked Particle RibbonWidth=1, SkeletalMesh=$System.Skeletal Mesh) {module=FN_UpdateAnimTrail001}
   &ScaleColor(
    Scale Alpha=&FloatFromCurve(
     FloatCurve=DataInterface() {propertyValues {Curve {keys [
         {arriveTangent=-2.003, interpMode=RCIM_Cubic, leaveTangent=-2.003, tangentMode=RCTM_User, time=0.5, value=1}
         {interpMode=RCIM_Cubic, tangentMode=RCTM_User, time=1, value=0}
        ]}}}
     Scale Curve=&FN_Select_Float_from_Time_Of_Day(Morning=1, Day=1, Evening=1, Night=1)
    )
    Scale RGB=&MakeVectorFromLinearColorRGB(Color=&ColorFromCurve(ColorCurve=DataInterface() {propertyValues {
        AlphaCurve {keys []}
        BlueCurve {keys [<at 0.077 1 RCIM_Cubic>,<at 0.9809 0.7>]}
        GreenCurve {keys [<at 0.077 1 RCIM_Cubic>,<at 0.9809 0.7>]}
        RedCurve {keys [<at 0.077 1 RCIM_Cubic>,<at 0.9809 0.7>]}
       }}))
    ScaleA=true
    ScaleRGB=false
   )
   &ScaleColor(
    Scale Alpha=&FloatFromCurve(FloatCurve=DataInterface() {propertyValues {Curve {keys [<at 0.2 1>,<at 0.3 0>]}}}, CurveIndex=$Emitter.Age)
    ScaleA=true
    ScaleRGB=true
   ) {module=ScaleColor001}
   &InheritSourceMovement(
    Applied Position Scale=&VectorFromFloat(Value=&FloatFromCurve(FloatCurve=DataInterface() {propertyValues {Curve {keys [<at 0.1 1>,<at 0.3 0>]}}}))
    ApplyForce=false
    ApplyPosition=true
    ApplyPositionFalloff=false
   ) {enabled=false}
   &SolveForcesAndVelocity(Clamp Velocity=false, Limit Acceleration=false) {enabled=false}
  ]
  rendererClasses [&NiagaraRibbonRendererProperties]
  renderers [{rendererClass=&NiagaraRibbonRendererProperties}]
 }
 TrailSphere {
  emitterUpdateScript [
   &EmitterState()
   &FN_SpawnAnimTrail(Spawn Interval=0)
  ]
  particleSpawnScript [
   &FN_InitializeAnimTrailBoneIndecies(First Bone or Socket Index=2, Second Bone or Socket Index=3)
   &FN_InitializeAnimTrail(Lifetime=0.06, SkeletalMesh=$System.Skeletal Mesh)
   /SetVariables(Particles.Color=&FNTimeOfDayColor(
     Night Color=<rgba 0.0525 0.0624 0.35 0.65>
     Morning Color=<rgba 0.1419 0.1576 0.33 0.65>
     Day Color=<rgba 0.129 0.1432 0.3 0.8>
     Evening Color=<rgba 0.1426 0.1565 0.31 0.65>
    )) {isSetParameters=true}
  ]
  particleUpdateScript [
   &FN_UpdateAnimTrail(Locked Particle RibbonWidth=1, SkeletalMesh=$System.Skeletal Mesh) {module=FN_UpdateAnimTrail001}
   &ScaleColor(
    Scale Alpha=&FloatFromCurve(
     FloatCurve=DataInterface() {propertyValues {Curve {keys [
         {arriveTangent=-2.2699, interpMode=RCIM_Cubic, leaveTangent=-2.2699, tangentMode=RCTM_User, time=0, value=1}
         {interpMode=RCIM_Cubic, tangentMode=RCTM_User, time=1, value=0}
        ]}}}
     Scale Curve=&FN_Select_Float_from_Time_Of_Day(Morning=1, Day=1, Evening=1, Night=1)
    )
    Scale RGB=&MakeVectorFromLinearColorRGB(Color=&ColorFromCurve(ColorCurve=DataInterface() {propertyValues {
        AlphaCurve {keys []}
        BlueCurve {keys [<at 0.0912 1 RCIM_Cubic>,<at 0.7115 0.8>]}
        GreenCurve {keys [<at 0.0912 1 RCIM_Cubic>,<at 0.7115 0.8>]}
        RedCurve {keys [<at 0.0912 1 RCIM_Cubic>,<at 0.7115 0.8>]}
       }}))
    ScaleA=true
    ScaleRGB=false
   )
   &ScaleColor(
    Scale Alpha=&FloatFromCurve(FloatCurve=DataInterface() {propertyValues {Curve {keys [<at 0.2 1>,<at 0.3 0>]}}}, CurveIndex=$Emitter.Age)
    ScaleA=true
    ScaleRGB=true
   ) {module=ScaleColor001}
  ]
  rendererClasses [&NiagaraRibbonRendererProperties]
  renderers [{rendererClass=&NiagaraRibbonRendererProperties}]
 }
}
summary {
 systemName=NS_Pickaxe_DirectContact_Trails
 emitters [
  {emitterName=TrailBlade, rendererClasses [&NiagaraRibbonRendererProperties]}
  {emitterName=TrailSphere, rendererClasses [&NiagaraRibbonRendererProperties]}
 ]
 userVariables [
  User.bisSwinging {type=NiagaraBool}
  User.RibbonIDOffset {type=NiagaraInt32}
  User.Variant {type=NiagaraFloat, defaultValue=0, description=""}
 ]
}
